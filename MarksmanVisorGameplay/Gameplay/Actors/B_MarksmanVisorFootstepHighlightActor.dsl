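; Footstep highlight actor for the Marksman Visor gameplay, a mesh whose look follows visor active, scanning and cooldown state.
; The construction script creates a dynamic material instance on the mesh and stores it.
; The SetState event toggles mesh visibility with the active flag, writes scan percent to the Progress parameter (zero when not scanning) and sets the Active parameter to zero during cooldown, one otherwise.
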(event Custom|SetState
  (params (IsActive :type Boolean) (IsCoolingDown :type Boolean) (IsScanning :type Boolean) (ScanPercent :type Float%20%28double-precision%29))
  (bind _materialinstance (Variables|Default|GetMaterialInstance))
  (Rendering|SetVisibility (Variables|Default|GetMesh) IsActive false)
  (Rendering|Material|SetScalarParameterValue _materialinstance "Progress" (Math|Float|SelectFloat ScanPercent 0.0 IsScanning) :!class /Script/Engine.MaterialInstanceDynamic)
  (Rendering|Material|SetScalarParameterValue _materialinstance "Active" (Math|Float|SelectFloat 0.0 1.0 IsCoolingDown) :!class /Script/Engine.MaterialInstanceDynamic))

(fn ConstructionScript
  (bind _returnvalue (Rendering|Material|CreateDynamicMaterialInstance (Variables|Default|GetMesh) 0 (Variables|Default|GetMaterial) "None" :!class /Script/Engine.PrimitiveComponent))
  (Variables|Default|SetMaterialInstance _returnvalue))
